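# Physics asset for the LeafyVest medium male player body cosmetic, set up for secondary motion on left and right pack flaps.
# Capsule bodies cover spine_05 and a base and dynamic bone for each flap.
# Each base flap is angularly locked to spine_05, acting as an anchor.
# Each dynamic flap hangs off its base with narrow swing limits and a zero twist limit, restricting it to slight flapping.
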PhysicsAsset /BRCosmetics/Characters/Player/Male/Medium/Bodies/M_MED_LeafyVest/Meshes/M_MED_LeafyVest_Physics
bodies {
 base_packflap_l [{shapeType=Capsule, length=0.505, radius=1, shapeName=base_packflap_l_capsule, rotation=<rot 90 0 0>}]
 base_packflap_r [{shapeType=Capsule, length=0.505, radius=1, shapeName=base_packflap_r_capsule, rotation=<rot 90 0 0>}]
 dyn_packflap_l [{shapeType=Capsule, length=1, radius=1, shapeName=dyn_packflap_l_capsule, center=<1 0 0>, rotation=<rot 90 0 5.9344>}]
 dyn_packflap_r [{shapeType=Capsule, length=1, radius=1, shapeName=dyn_packflap_r_capsule, center=<1 0 0>, rotation=<rot 90 0 8.8066>}]
 spine_05 [{shapeType=Capsule, length=15, radius=10, shapeName=spine_05_capsule}]
}
constraints [
 {bone1Name=dyn_packflap_l, bone2Name=base_packflap_l, swing1LimitDegrees=10, swing2LimitDegrees=20, twistLimitDegrees=0}
 {bone1Name=dyn_packflap_r, bone2Name=base_packflap_r, swing1LimitDegrees=10, swing2LimitDegrees=20, twistLimitDegrees=0}
 {
  bone1Name=base_packflap_r
  bone2Name=spine_05
  swing1LimitDegrees=45
  swing1Motion=Locked
  swing2LimitDegrees=45
  swing2Motion=Locked
  twistLimitDegrees=45
  twistMotion=Locked
 }
 {
  bone1Name=base_packflap_l
  bone2Name=spine_05
  swing1LimitDegrees=45
  swing1Motion=Locked
  swing2LimitDegrees=45
  swing2Motion=Locked
  twistLimitDegrees=45
  twistMotion=Locked
 }
]
